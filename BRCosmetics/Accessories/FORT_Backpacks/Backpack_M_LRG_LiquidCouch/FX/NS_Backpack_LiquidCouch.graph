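# Niagara effect for the LiquidCouch backpack cosmetic that leaks bursts of toon-shaded mesh particles from its sockets.
# Each MainLeak loop picks a socket from a shuffled-offset order array, which selects spawn rate, lifetime, size and path.
# Particles follow the sampled socket with interpolated lag blended over age, adjusted for movement and skydiving.
# Color follows time of day, curves shape each burst, and a debug sprite emitter is disabled.
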NiagaraSystem /BRCosmetics/Accessories/FORT_Backpacks/Backpack_M_LRG_LiquidCouch/FX/NS_Backpack_LiquidCouch
dependencies {
 usedDataInterfaces [
  &NiagaraDataInterfacePlatformSet
  &NiagaraDataInterfaceCurve
  &NiagaraDataInterfaceVectorCurve
  &NiagaraDataInterfaceSocketReader
  &NiagaraDataInterfaceArrayInt32
 ]
 usedDynamicInputs [
  &Vector2DFromFloat
  &MaskFloatBySpawnGroup
  &NormalizedAngleToDegrees
  &Boolean_OrOperation
  &SetBoolByFloatComparison
  &Abs_Float
 ]
 usedModules [
  &self:SystemSpawnScript.SystemScriptSource.SystemScriptGraph.NiagaraNodeAssignment_1.SetVariables_#1
  &SystemState
  &self:SystemSpawnScript.SystemScriptSource.SystemScriptGraph.NiagaraNodeAssignment_3.SetVariables_#2
  &self:SystemSpawnScript.SystemScriptSource.SystemScriptGraph.NiagaraNodeAssignment_2.SetVariables_#3
  &self:SystemSpawnScript.SystemScriptSource.SystemScriptGraph.NiagaraNodeAssignment_4.SetVariables_#4
 ]
}
emitters {
 FN_Debug001 {
  bEnabled=false
  emitterSpawnScript [/SetVariables(Emitter.rand:NiagaraInt32=&RandomRangeInt(Minimum=5, Maximum=7)) {isSetParameters=true}]
  emitterUpdateScript [
   &EmitterState(Life Cycle Mode=::NewEnumerator1, Loop Delay=5, Loop Duration=2, UseLoopDelay=false)
   &SpawnBurst_Instantaneous(Spawn Count=1, Use Spawn Probability=false)
  ]
  particleSpawnScript [/Niagara/Modules/Spawn/Initialization/InitializeParticle(
    Lifetime=10
    Module.Write Color=true
    Module.Write Lifetime=true
    Module.Write Mass=true
    Module.Write Material Random=true
    Module.Write Position=false
    Module.Write Scale=false
    Module.Write SpriteRotation=false
    Module.Write SpriteSize=true
    Position=&SimulationPosition()
    Sprite Size=<60 60>
   )]
  particleUpdateScript [
   &ParticleState()
   /SetVariables(Particles.Debug:NiagaraFloat=$System.NormalizedLoopAge) {isSetParameters=true}
   &DynamicMaterialParameters(
    Index 0 Param 1=$Particles.Debug
    Module.Param0WriteEnabled=true
    Module.Param1WriteEnabled=true
    Module.Param2WriteEnabled=true
    Module.Param3WriteEnabled=true
   )
   &SystemLocation(MaskSpawnGroup=false, Offset=<-50 0 50>)
  ]
  rendererClasses [&NiagaraSpriteRendererProperties]
  renderers [{rendererClass=&NiagaraSpriteRendererProperties}]
 }
 MainLeak {
  emitterSpawnScript [
   /Game/Effects/Niagara/MasterMaterials/ToonShader/Template/NM_OutlineParameters(OutlineColor:LinearColor=<rgba 0.25 0.25 0.25 1>)
   /Game/Effects/Niagara/MasterMaterials/ToonShader/Template/NM_ToonShaderParameters(HighLight_Softness=1)
   /SetVariables(Emitter.SocketOrderRandomOffset:NiagaraInt32=&RandomRangeInt(Minimum=0, Maximum=3)) {isSetParameters=true}
  ]
  emitterUpdateScript [
   &EmitterState(
    Life Cycle Mode=::NewEnumerator1
    Loop Delay=&RandomRangeFloat(Minimum=5, Maximum=9)
    Loop Duration=&RandomRangeFloat(Minimum=3, Maximum=4)
    Recalculate Duration Each Loop=true
    UseLoopDelay=true
   )
   /SetVariables(Emitter.RandSocket:NiagaraInt32=&SelectIntFromArray(
     Direct Array Index=&Add_Integer(A=$Emitter.LoopCount, B=$Emitter.SocketOrderRandomOffset)
     Direct Array Mode=::NewEnumerator1
     Int Selection Array=$System.SocketOrder
    )) {isSetParameters=true}
   &SpawnRate(
    SpawnRate=&Multiply_Float(
     A=&Multiply_Float(
      A=&MergeHelper_FloatSwitch(
       0=20
       1=20
       2=&Lerp_Float(A=80, B=90, Alpha=$System.IsMovingInterpolated)
       3=&Lerp_Float(A=80, B=90, Alpha=$System.IsMovingInterpolated)
       4=0
       Index=$Emitter.RandSocket
       Num=::NewEnumerator3
      )
      B=&FloatFromCurve(
       FloatCurve=DataInterface() {propertyValues {Curve {keys [<at 0 1 RCIM_Constant>,<at 0.1663 0.3 RCIM_Constant>,<at 0.4288 0.8635 RCIM_Constant>,<at 0.6128 0.6259 RCIM_Constant>,<at 0.838 1 RCIM_Constant>,<at 1 1 RCIM_Constant>]}}}
       CurveIndex=$System.NormalizedLoopAge
      )
     )
     B=1
    )
    Use Spawn Probability=false
   )
   /SetVariables(Emitter.FELight:LinearColor=&MakeCustomLinearColorFromBool(Bool=$User.IsFrontEnd, True Color=<rgba 0 0 1 1>, False Color=<rgba 0 0 0 0>)) {isSetParameters=true}
  ]
  particleSpawnScript [
   /SetVariables(Particles.BoneIndex:NiagaraInt32=$Emitter.RandSocket) {isSetParameters=true}
   /Niagara/Modules/Spawn/Initialization/V2/InitializeParticle(
    Lifetime=&Multiply_Float(
     A=&MergeHelper_FloatSwitch(
      0=&RandomRangeFloat(Minimum=0.8, Maximum=&RandomRangeFloat(Minimum=0.8, Maximum=1.1))
      1=&RandomRangeFloat(Minimum=0.8, Maximum=&RandomRangeFloat(Minimum=0.8, Maximum=1.1))
      2=&Lerp_Float(A=&RandomRangeFloat(Minimum=0.75, Maximum=0.8), B=0.4, Alpha=$System.IsMovingInterpolated)
      3=&Lerp_Float(A=&RandomRangeFloat(Minimum=0.75, Maximum=0.8), B=0.4, Alpha=$System.IsMovingInterpolated)
      Index=$StackContext.BoneIndex
      Num=::NewEnumerator2
     )
     B=0.75
    )
    Mass Mode=::NewEnumerator1
    Mesh Scale Mode=::NewEnumerator3
    Mesh Uniform Scale=&Multiply_Float(A=&RandomRangeFloat(Minimum=0.5, Maximum=0.7), B=0.004)
    Sprite Rotation Mode=::NewEnumerator2
    UsePositionOffset=false
    Write Material Random=true
   ) {module=InitializeParticle002}
   /SetVariables(Particles.InitSpriteRotRate=&UniformRangedFloat(Minimum=-3, Maximum=3)) {isSetParameters=true}
   &InitialMeshOrientation(Rotation=&RandomRangeVector(Minimum=<0 0 0>, Maximum=<1 1 1>), Use Rotation Vector=false)
  ]
  particleUpdateScript [
   &ParticleState()
   &UpdateMeshOrientation(Rotation Rate=0.3, Rotation Vector=&RandomRangeVector(Minimum=<1 0 0>, Maximum=<1 0 0>))
   &SocketLocation(
    Orientation Sampling=::NewEnumerator1
    Position Sampling=::NewEnumerator1
    SamplePosOffset=false
    Socket Index=&MergeHelper_IntSwitch(0=0, 1=1, 2=2, 3=3, 4=?, 5=?, Index=$Particles.BoneIndex, Num=::NewEnumerator2)
    Socket Reader=$System.Socket Reader
    Socket Sampling Mode=::NewEnumerator3
    VelocitySpeedLimit=false
   ) {module=SocketLocation001}
   &ShapeLocation(
    MaskSpawnGroup=false
    Override Seed=false
    Shape Origin=$Output.SocketLocation001.SampledPosition
    Sphere Radius=&MergeHelper_FloatSwitch(0=1, 1=1, 2=2.5, 3=2.5, Index=$Particles.BoneIndex, Num=::NewEnumerator2)
    UseOwnerScale=true
   )
   &InterpolateOverTime(
    Rate Of Change:NiagaraFloat=&MakeCustomFloatFromBool(Bool=$User.IsSkydiving, True Float=30, False Float=&Lerp_Float(A=15, B=20, Alpha=$System.IsMovingInterpolated))
    Target Position Value:NiagaraPosition=$Particles.ShapeLocation.ShapePosition
    Variable Type=::NewEnumerator6
   )
   /SetVariables(Particles.Position=&LerpPosition(
     A=$Particles.ShapeLocation.ShapePosition
     B=&MakeVector(
      X=&MakeFloatFromVector(
       Vector=Linked() {linkedVariable {name=StackContext.InterpolateOverTime.InterpolateOverTime_Position.Moving Average, type=NiagaraPosition}}
       Channel=::NewEnumerator0
      )
      Y=&MakeFloatFromVector(
       Vector=Linked() {linkedVariable {name=StackContext.InterpolateOverTime.InterpolateOverTime_Position.Moving Average, type=NiagaraPosition}}
       Channel=::NewEnumerator1
      )
      Z=&MakeFloatFromVector(Vector=Linked() {linkedVariable {name=Output.SocketLocation001.SampledPosition, type=NiagaraPosition}}, Channel=::NewEnumerator2)
     )
     Alpha=$Particles.NormalizedAge
    )) {isSetParameters=true}
   /SetVariables(Particles.Position=&AddVectorToPosition(
     Position=$StackContext.Position
     Vector=&RotateVectorByQuaternion(
      Vector To Rotate=&MakeVector(
       X=&Lerp_Float(
        A=-0.05
        B=&Multiply_Float(
         A=&Lerp_Float(
          A=&UniformAOrBFloat(
           A=&RandomRangeFloat(Minimum=5, Maximum=15)
           B=25
           Distribution Weight=0.1
           Evaluation Type=::NewEnumerator0
           Recalculate Random Each Loop=true
           Override Seed=false
          )
          B=&UniformAOrBFloat(
           A=&RandomRangeFloat(Minimum=4, Maximum=8)
           B=8
           Distribution Weight=0.1
           Evaluation Type=::NewEnumerator0
           Recalculate Random Each Loop=true
           Override Seed=false
          )
          Alpha=$System.IsMovingInterpolated
         )
         B=&FN_PickFloatDependOnPlatform(TargetPlatformActive=1, TargetPlatformNoActive=0.5)
        )
        Alpha=$Particles.NormalizedAge
       )
       Y=0
       Z=0
      )
      Quaternion=$Output.SocketLocation001.SampledRotation
     )
    )) {isSetParameters=true}
   /SetVariables(Particles.Position=&MergeHelper_VectorSwitch(
     0=&AddVectorToPosition(
      Position=$StackContext.ShapeLocation.ShapePosition
      Vector=&RotateVectorByQuaternion(
       Vector To Rotate=&VectorFromCurve(
        VectorCurve=DataInterface() {propertyValues {
          XCurve {keys [<at 0 0>,<at 1 1>]}
          YCurve {keys [<at 0 0>,<at 1 1>]}
          ZCurve {keys [<at 0 0 RCIM_Cubic>, {arriveTangent=1.5915, interpMode=RCIM_Cubic, leaveTangent=1.5915, tangentMode=RCTM_User, time=1, value=1}]}
         }}
        Scale Curve=&RandomRangeVector(Minimum=<0 -5 -4>, Maximum=<0 -5 -5>)
       )
       Quaternion=&MakeQuaternion(
        Quaternion From=0
        Axis=<0 0 1>
        Angle=&RandomRangeFloat(Minimum=-0.08, Maximum=-0.12)
        From=?
        To=?
        Angle Type=::NewEnumerator1
        Coordinate Space=::Local
        Normalize Vectors=false
       )
      )
     )
     1=&AddVectorToPosition(
      Position=$StackContext.ShapeLocation.ShapePosition
      Vector=&RotateVectorByQuaternion(
       Vector To Rotate=&VectorFromCurve(
        VectorCurve=DataInterface() {propertyValues {
          XCurve {keys [<at 0 0>,<at 1 1>]}
          YCurve {keys [<at 0 0>,<at 1 1>]}
          ZCurve {keys [<at 0 0 RCIM_Cubic>, {arriveTangent=1.5915, interpMode=RCIM_Cubic, leaveTangent=1.5915, tangentMode=RCTM_User, time=1, value=1}]}
         }}
        Scale Curve=&RandomRangeVector(Minimum=<0 -5 -4>, Maximum=<0 -5 -5>)
       )
       Quaternion=&MakeQuaternion(
        Quaternion From=0
        Axis=<0 0 1>
        Angle=&RandomRangeFloat(Minimum=0.08, Maximum=0.12)
        From=?
        To=?
        Angle Type=::NewEnumerator1
        Coordinate Space=::Local
        Normalize Vectors=false
       )
      )
     )
     2=Linked() {linkedVariable {name=StackContext.Position, type=NiagaraPosition}}
     3=Linked() {linkedVariable {name=StackContext.Position, type=NiagaraPosition}}
     Index=$Particles.BoneIndex
     Num=::NewEnumerator2
    )) {isSetParameters=true}
   &SolveForcesAndVelocity(Clamp Velocity=false, Limit Acceleration=false)
   &ScaleColor(
    Scale Alpha=&FloatFromCurve(FloatCurve=DataInterface() {propertyValues {Curve {keys [{arriveTangent=-1.3064, interpMode=RCIM_Cubic, leaveTangent=-1.3064, tangentMode=RCTM_User, time=0, value=1}, <at 1 0 RCIM_Cubic>]}}})
    Scale RGB=&MakeVectorFromLinearColorRGB(Color=&FNTimeOfDayColor(
      Night Color=<rgba 0.1937 0.4023 0.6458 1>
      Morning Color=<rgba 0.9896 0.5937 0.7171 1>
      Day Color=<rgba 1 1 1 1>
      Evening Color=<rgba 0.4 0.1937 0.2243 1>
     ))
    ScaleA=true
    ScaleRGB=true
   )
   &ScaleMeshSize(Scale Factor=&VectorFromFloat(Value=&MergeHelper_FloatSwitch(
      0=&FloatFromCurve(FloatCurve=DataInterface() {propertyValues {Curve {keys [
           {arriveTangent=6.0769, interpMode=RCIM_Cubic, leaveTangent=6.0769, tangentMode=RCTM_User, time=0.001, value=0.8035}
           {arriveTangent=0.4607, interpMode=RCIM_Cubic, leaveTangent=0.4607, time=0.1, value=1.8527}
           {arriveTangent=-2.0585, interpMode=RCIM_Cubic, leaveTangent=-2.0585, time=0.7218, value=1.1356}
           {arriveTangent=-2.8143, tangentMode=RCTM_User, time=1, value=0}
          ]}}})
      1=&FloatFromCurve(FloatCurve=DataInterface() {propertyValues {Curve {keys [
           {arriveTangent=6.0769, interpMode=RCIM_Cubic, leaveTangent=6.0769, tangentMode=RCTM_User, time=0.001, value=0.8035}
           {arriveTangent=0.4607, interpMode=RCIM_Cubic, leaveTangent=0.4607, time=0.1, value=1.8527}
           {arriveTangent=-2.0585, interpMode=RCIM_Cubic, leaveTangent=-2.0585, time=0.7218, value=1.1356}
           {arriveTangent=-2.8143, tangentMode=RCTM_User, time=1, value=0}
          ]}}})
      2=&FloatFromCurve(FloatCurve=DataInterface() {propertyValues {Curve {keys [
           {arriveTangent=1.208, interpMode=RCIM_Cubic, leaveTangent=1.208, tangentMode=RCTM_User, time=0, value=1.767}
           {arriveTangent=-0.8748, interpMode=RCIM_Cubic, leaveTangent=-0.8748, time=0.3313, value=2.1354}
           {arriveTangent=-3.1669, interpMode=RCIM_Cubic, leaveTangent=-3.1669, tangentMode=RCTM_User, time=0.7218, value=1.1356}
           {arriveTangent=-4.5807, tangentMode=RCTM_User, time=1, value=0}
          ]}}})
      3=&FloatFromCurve(FloatCurve=DataInterface() {propertyValues {Curve {keys [
           {arriveTangent=1.208, interpMode=RCIM_Cubic, leaveTangent=1.208, tangentMode=RCTM_User, time=0, value=1.767}
           {arriveTangent=-0.8748, interpMode=RCIM_Cubic, leaveTangent=-0.8748, time=0.3313, value=2.1354}
           {arriveTangent=-3.1669, interpMode=RCIM_Cubic, leaveTangent=-3.1669, tangentMode=RCTM_User, time=0.7218, value=1.1356}
           {arriveTangent=-4.5807, tangentMode=RCTM_User, time=1, value=0}
          ]}}})
      Index=$Particles.BoneIndex
      Num=::NewEnumerator2
     )))
   /SetVariables(Particles.DP0:Vector4f=&MakeVector4(
     X=5
     Y=&FloatFromCurve(FloatCurve=DataInterface() {propertyValues {Curve {keys [
          {arriveTangent=-0.7948, interpMode=RCIM_Cubic, leaveTangent=-0.7949, tangentMode=RCTM_User, time=0.2519, value=0.1057}
          <at 0.7709 0 RCIM_Cubic>
         ]}}})
     Z=0
     W=&FloatFromCurve(FloatCurve=DataInterface() {propertyValues {Curve {keys [<at 0 1 RCIM_Cubic>,<at 0.7 1 RCIM_Cubic>]}}})
    )) {isSetParameters=true}
   &CameraOffset(Camera Offset Amount=&MergeHelper_FloatSwitch(
     0=1
     1=1
     2=&FN_FrontEndFloatModifier(Gameplay=&Lerp_Float(A=-1, B=-2, Alpha=$System.IsMovingInterpolated), Front End=-1)
     3=&FN_FrontEndFloatModifier(Gameplay=&Lerp_Float(A=-1, B=-2, Alpha=$System.IsMovingInterpolated), Front End=-1)
     Index=$Particles.BoneIndex
     Num=::NewEnumerator2
    ))
   &ScaleMeshSize(Scale Factor=&VectorFromFloat(Value=&FloatFromCurve(FloatCurve=DataInterface() {propertyValues {Curve {keys [<at 0.04 0>,<at 0.05 1>]}}}))) {module=ScaleMeshSize001}
   &ScaleMeshSize(Scale Factor=&VectorFromFloat(Value=&FloatFromCurve(
      FloatCurve=DataInterface() {propertyValues {Curve {
         defaultValue=0
         keys [
          <at 0 0 RCIM_Cubic>
          {arriveTangentWeight=0.0581, interpMode=RCIM_Cubic, leaveTangentWeight=0.3488, tangentMode=RCTM_User, time=0.1, value=1}
          {arriveTangentWeight=0.3008, interpMode=RCIM_Cubic, leaveTangentWeight=0.1454, tangentMode=RCTM_User, time=0.7, value=1}
          <at 0.99 0 RCIM_Cubic>
         ]
        }}}
      CurveIndex=&ClampFloat(Float=$Emitter.NormalizedLoopAge)
     ))) {module=ScaleMeshSize003}
  ]
  rendererClasses [&NiagaraMeshRendererProperties]
  renderers [
   {rendererClass=&NiagaraMeshRendererProperties, rendererIndex=0}
   {rendererClass=&NiagaraMeshRendererProperties, rendererIndex=1}
  ]
 }
}
summary {
 systemName=NS_Backpack_LiquidCouch
 emitters [
  {emitterName=MainLeak, rendererClasses [&NiagaraMeshRendererProperties]}
  {bEnabled=false, emitterName=FN_Debug001, rendererClasses [&NiagaraSpriteRendererProperties]}
 ]
 userVariables [
  User.bisFrontEnd {type=NiagaraBool}
  User.bIsFrontEndPreview {type=NiagaraBool}
  User.IsFrontEnd {type=NiagaraBool}
  User.IsSkydiving {type=NiagaraBool}
 ]
}
