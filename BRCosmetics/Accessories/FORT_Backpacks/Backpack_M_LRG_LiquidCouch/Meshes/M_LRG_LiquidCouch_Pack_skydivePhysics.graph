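# Skydive physics asset for the LiquidCouch backpack cosmetic, simulating its left and right extinguishers and tubes.
# Three capsules form the pack body, tiny capsules act as base anchors, and larger capsules cover each dynamic extinguisher and tube bone.
# Each base joins the pack with free rotation and carries its dynamic bone.
# Extinguishers swing within angular limits, while tube joints are rotationally locked to their bases despite having limit values set.
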PhysicsAsset /BRCosmetics/Accessories/FORT_Backpacks/Backpack_M_LRG_LiquidCouch/Meshes/M_LRG_LiquidCouch_Pack_skydivePhysics
bodies {
 base_extinguisher_l [{shapeType=Capsule, length=0.505, radius=0.505, shapeName=base_extinguisher_l_capsule, rotation=<rot 0 -90 0>}]
 base_extinguisher_r [{shapeType=Capsule, length=0.505, radius=0.505, shapeName=base_extinguisher_r_capsule, rotation=<rot 0 -90 0>}]
 base_tube_l [{shapeType=Capsule, length=0.505, radius=0.505, shapeName=base_tube_l_capsule, rotation=<rot 0 -90 0>}]
 base_tube_r [{shapeType=Capsule, length=0.505, radius=0.505, shapeName=base_tube_r_capsule, rotation=<rot 0 -90 0>}]
 dyn_extinguisher_l [{shapeType=Capsule, length=3.2248, radius=4.2024, shapeName=dyn_extinguisher_l_capsule, center=<9 0 0>, rotation=<rot 90 0 0>}]
 dyn_extinguisher_r [{shapeType=Capsule, length=3.0655, radius=3.9881, shapeName=dyn_extinguisher_r_capsule, center=<9 -0.5 0>, rotation=<rot 90 0 0>}]
 dyn_tube_l [{shapeType=Capsule, length=14.1781, radius=2, shapeName=dyn_tube_l_capsule, center=<2 0 6>, rotation=<rot 79.9374 12.2 -9.623>}]
 dyn_tube_r [{
   shapeType=Capsule
   length=12.9893
   radius=1.8294
   shapeName=dyn_tube_r_capsule
   center=<1.863 -4.4907 2.137>
   rotation=<rot 49.9107 3.032 23.7885>
  }]
 pack [
  {shapeType=Capsule, length=20, radius=6, shapeName=pack_capsule, center=<0 -12.3735 -9.7121>, rotation=<rot 0 -90 0>}
  {shapeType=Capsule, length=24.4441, radius=5.5559, shapeName=pack_capsule, center=<-11.3598 -12.4816 -3.8777>}
  {shapeType=Capsule, length=22.2166, radius=5.6463, shapeName=pack_capsule, center=<10.4488 -12.5664 -3.8604>}
 ]
}
constraints [
 {bone1Name=dyn_extinguisher_l, bone2Name=base_extinguisher_l, swing1LimitDegrees=25, swing2LimitDegrees=45, twistLimitDegrees=5}
 {bone1Name=dyn_extinguisher_r, bone2Name=base_extinguisher_r, swing1LimitDegrees=25, swing2LimitDegrees=45, twistLimitDegrees=5}
 {
  bone1Name=base_extinguisher_l
  bone2Name=pack
  swing1LimitDegrees=45
  swing1Motion=Free
  swing2LimitDegrees=45
  swing2Motion=Free
  twistLimitDegrees=45
  twistMotion=Free
 }
 {
  bone1Name=base_extinguisher_r
  bone2Name=pack
  swing1LimitDegrees=45
  swing1Motion=Free
  swing2LimitDegrees=45
  swing2Motion=Free
  twistLimitDegrees=45
  twistMotion=Free
 }
 {
  bone1Name=dyn_tube_r
  bone2Name=base_tube_r
  swing1LimitDegrees=20
  swing1Motion=Locked
  swing2LimitDegrees=8
  swing2Motion=Locked
  twistLimitDegrees=5
  twistMotion=Locked
 }
 {
  bone1Name=base_tube_l
  bone2Name=pack
  swing1LimitDegrees=45
  swing1Motion=Free
  swing2LimitDegrees=45
  swing2Motion=Free
  twistLimitDegrees=45
  twistMotion=Free
 }
 {
  bone1Name=dyn_tube_l
  bone2Name=base_tube_l
  swing1LimitDegrees=17.5
  swing1Motion=Locked
  swing2LimitDegrees=5
  swing2Motion=Locked
  twistLimitDegrees=5
  twistMotion=Locked
 }
 {
  bone1Name=base_tube_r
  bone2Name=pack
  swing1LimitDegrees=45
  swing1Motion=Free
  swing2LimitDegrees=45
  swing2Motion=Free
  twistLimitDegrees=45
  twistMotion=Free
 }
]
